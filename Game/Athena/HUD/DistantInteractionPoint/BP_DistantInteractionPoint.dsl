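; BP_DistantInteractionPoint is an actor that displays an interaction point widget for the Athena HUD, with distance tracking.
; A helper function gets the user widget object from its InteractionPointWidget and casts it to BP_DistantInteractionPointWidget.
; On BeginPlay, the actor sets itself as that widget's target actor, then starts the widget's timer for updating distance.
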(event EventBeginPlay
  (bind _interactionpointwidget (CallFunction|GetInteractionPointWidget))
  ;[0.08818 0.599 0.07413] Set widgets target and start its update timer for distance calculation
  (Variables|/Script/FortniteGame.InteractionPointWidget|SetTargetActor self _interactionpointwidget)
  (Class|InteractionPointWidget|StartDistanceUpdateTimer _interactionpointwidget))

(fn GetInteractionPointWidget
  ;[1 0.7967 0.1307] Get the widget this actor is displaying
  (bind _asbp_distant_interaction_point_widget (Utilities|Casting|CastToBP_DistantInteractionPointWidget (UserInterface|GetUserWidgetObject (Variables|Default|GetInteractionPointWidget))))
  (return _asbp_distant_interaction_point_widget))

(fn ConstructionScript)
