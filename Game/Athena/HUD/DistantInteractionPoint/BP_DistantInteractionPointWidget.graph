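# HUD widget for an interaction point marker, built on the native InteractionPointWidget, with near and distant looks.
# An inherited switcher holds a pulsing ring with the near icon, and a distant view with arrow, icon and distance text.
# PreConstruct applies the two icon texture variables to the images, and Set Icon Images stores new textures and updates both brushes.
# No graph here drives switching or the distance text.
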WidgetBlueprint /Game/Athena/HUD/DistantInteractionPoint/BP_DistantInteractionPointWidget
graphs [
 EventGraph {nodes [
   Event PreConstruct()
   > SetBrushFromTexture!(
    execute
    self:Image Object Reference=Get Image_Icon() -> (Image_Icon:Image Object Reference)
    Texture:Texture 2D Object Reference=Get NearIconTexture() -> (NearIconTexture:Texture 2D Object Reference)
    bMatchSize:bool=false
   ) -> (then)
   > SetBrushFromTexture(
    self=Get Image_DistantIcon() -> (Image_DistantIcon:Image Object Reference)
    Texture=Get DistantIconTexture() -> (DistantIconTexture:Texture 2D Object Reference)
    bMatchSize=false
   )
  ]}
 Set Icon Images {nodes [
   n1 = Set Icon Images!() -> (then, NearIcon:Texture 2D Object Reference, DistantIcon:Texture 2D Object Reference) {is=title K2Node_FunctionEntry}
   > n2 = Set NearIconTexture(NearIconTexture:Texture 2D Object Reference=@n1) -> (Output_Get:Texture 2D Object Reference)
   > SetBrushFromTexture(self=Get Image_Icon(), Texture=@n2, bMatchSize=false)
   > n3 = Set DistantIconTexture(DistantIconTexture:Texture 2D Object Reference=@n1.DistantIcon) -> (Output_Get:Texture 2D Object Reference)
   > SetBrushFromTexture(self=Get Image_DistantIcon(), Texture=@n3, bMatchSize=false)
  ]}
]
variables [
 DistantIconTexture {type=Texture2D, default=&/T-UI-InteractionPointIcon}
 NearIconTexture {type=Texture2D, default=&T-UI-InteractionPointIcon}
]
widgets {
 info {inheritedWidgetCount=2, parentClass=/Script/FortniteGame.InteractionPointWidget, rootWidgetClass=&ScaleBox, widgetCount=12}
 widgets [
  ScaleBox_Container {
   widgetClassPath=&ScaleBox
   bIsVariable=true
   widget=ScaleBox_Container
   properties {stretch=UserSpecified, userSpecifiedScale=0.75}
  }
  Switcher_Icons {
   widgetClassPath=&CommonWidgetSwitcherLegacy
   bInherited=true
   bIsVariable=true
   parent=ScaleBox_Container
   widget=Switcher_Icons
   slot {class=ScaleBoxSlot}
  }
  ScaleBox_InteractionIndication {
   widgetClassPath=&ScaleBox
   bIsVariable=true
   parent=Switcher_Icons
   widget=ScaleBox_InteractionIndication
   properties {stretch=UserSpecified, userSpecifiedScale=1.25}
   slot {class=WidgetSwitcherSlot}
  }
  Overlay_Ping {widgetClassPath=&Overlay, parent=ScaleBox_InteractionIndication, widget=Overlay_Ping, slot {class=ScaleBoxSlot}}
  Image_Ping {
   widgetClassPath=&Image
   bIsVariable=true
   parent=Overlay_Ping
   widget=Image_Ping
   properties {brush {
     bIsDynamicallyLoaded=false
     drawAs=Image
     imageType=NoImage
     mirroring=NoMirror
     resourceName=None
     resourceObject=/Game/UI/Login/SubGameSelect/Materials/M_UI_FocusBorderPulse_Circle_Inst
     tiling=NoTile
     imageSize=<32 32>
     margin=<ltrb 0 0 0 0>
     outlineSettings {
      bUseBrushTransparency=false
      roundingType=HalfHeightRadius
      width=0
      color {colorUseRule=UseColor_Specified, specifiedColor=<rgba 0 0 0 0>}
      cornerRadii=<0 0 0 0>
     }
     tintColor {colorUseRule=UseColor_Specified, specifiedColor=<rgba 1 1 1 1>}
     uVRegion {bIsValid=false, max=<0 0>, min=<0 0>}
    }}
   slot {class=OverlaySlot, properties {horizontalAlignment=HAlign_Fill, verticalAlignment=VAlign_Fill}}
  }
  Image_Icon {
   widgetClassPath=&Image
   bIsVariable=true
   parent=Overlay_Ping
   widget=Image_Icon
   properties {brush {
     bIsDynamicallyLoaded=false
     drawAs=Image
     imageType=NoImage
     mirroring=NoMirror
     resourceName=None
     resourceObject=&T-UI-InteractionPointIcon
     tiling=NoTile
     imageSize=<64 64>
     margin=<ltrb 0 0 0 0>
     outlineSettings {
      bUseBrushTransparency=false
      roundingType=HalfHeightRadius
      width=0
      color {colorUseRule=UseColor_Specified, specifiedColor=<rgba 0 0 0 0>}
      cornerRadii=<0 0 0 0>
     }
     tintColor {colorUseRule=UseColor_Specified, specifiedColor=<rgba 1 1 1 1>}
     uVRegion {bIsValid=false, max=<0 0>, min=<0 0>}
    }}
   slot {class=OverlaySlot, properties {horizontalAlignment=HAlign_Center, verticalAlignment=VAlign_Center}}
  }
  Overlay_31 {widgetClassPath=&Overlay, parent=Switcher_Icons, widget=Overlay_31, slot {class=WidgetSwitcherSlot}}
  Image_Arrow {
   widgetClassPath=&Image
   bIsVariable=true
   parent=Overlay_31
   widget=Image_Arrow
   properties {brush {
     bIsDynamicallyLoaded=false
     drawAs=Image
     imageType=NoImage
     mirroring=NoMirror
     resourceName=None
     resourceObject=/Game/Athena/HUD/Materials/M_UI_FocusBorderPulse_ArrowOnly
     tiling=NoTile
     imageSize=<128 128>
     margin=<ltrb 0 0 0 0>
     outlineSettings {
      bUseBrushTransparency=false
      roundingType=HalfHeightRadius
      width=0
      color {colorUseRule=UseColor_Specified, specifiedColor=<rgba 0 0 0 0>}
      cornerRadii=<0 0 0 0>
     }
     tintColor {colorUseRule=UseColor_Specified, specifiedColor=<rgba 1 1 1 1>}
     uVRegion {bIsValid=false, max=<0 0>, min=<0 0>}
    }}
   slot {class=OverlaySlot, properties {horizontalAlignment=HAlign_Center, verticalAlignment=VAlign_Center}}
  }
  Image_DistantIcon {
   widgetClassPath=&Image
   bIsVariable=true
   parent=Overlay_31
   widget=Image_DistantIcon
   properties {brush {
     bIsDynamicallyLoaded=false
     drawAs=Image
     imageType=NoImage
     mirroring=NoMirror
     resourceName=None
     resourceObject=null
     tiling=NoTile
     imageSize=<64 64>
     margin=<ltrb 0 0 0 0>
     outlineSettings {
      bUseBrushTransparency=false
      roundingType=HalfHeightRadius
      width=0
      color {colorUseRule=UseColor_Specified, specifiedColor=<rgba 0 0 0 0>}
      cornerRadii=<0 0 0 0>
     }
     tintColor {colorUseRule=UseColor_Specified, specifiedColor=<rgba 1 1 1 1>}
     uVRegion {bIsValid=false, max=<0 0>, min=<0 0>}
    }}
   slot {class=OverlaySlot, properties {horizontalAlignment=HAlign_Center, verticalAlignment=VAlign_Center}}
  }
  HorizontalBox_0 {
   widgetClassPath=&HorizontalBox
   parent=Overlay_31
   widget=HorizontalBox_0
   slot {class=OverlaySlot, properties {horizontalAlignment=HAlign_Fill, verticalAlignment=VAlign_Center}}
  }
  Spacer_69 {
   widgetClassPath=&Spacer
   parent=HorizontalBox_0
   widget=Spacer_69
   slot {class=HorizontalBoxSlot, properties {size {sizeRule=Fill, value=2}}}
  }
  TextBlock_Distance {
   widgetClassPath=&CommonTextBlock
   bInherited=true
   bIsVariable=true
   parent=HorizontalBox_0
   widget=TextBlock_Distance
   properties {
    text=1000 M
    font {
     bForceMonospaced=false
     bMaterialIsStencil=false
     fontMaterial=null
     fontObject=&Burbank
     letterSpacing=0
     monospacedWidth=1
     size=22
     skewAmount=0
     typefaceFontName=Burbank-PlayerPower
     outlineSettings {
      bApplyOutlineToDropShadows=false
      bMiteredCorners=false
      bSeparateFillAlpha=false
      outlineMaterial=null
      outlineSize=1
      outlineColor=<rgba 0 0 0 1>
     }
    }
    shadowOffset=<0 0>
   }
   slot {class=HorizontalBoxSlot, properties {padding=<ltrb 30 0 0 0>, size {sizeRule=Fill, value=1}}}
  }
 ]
}
